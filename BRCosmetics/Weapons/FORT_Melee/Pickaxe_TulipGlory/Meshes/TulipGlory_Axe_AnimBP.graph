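# Animation Blueprint for the TulipGlory pickaxe that adds springy, zero-gravity secondary motion to three upper blade bones.
# The input pose feeds three differently tuned AnimDynamics chains, cached as main, equipped and unequipped.
# In the front end it picks equipped or unequipped by IsHoldingPickaxe, otherwise it uses main.
# IsHoldingPickaxe follows the validity of BodyMesh, which is cached when character customization loads.
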AnimBlueprint /BRCosmetics/Weapons/FORT_Melee/Pickaxe_TulipGlory/Meshes/TulipGlory_Axe_AnimBP
target_skeleton=&/TulipGlory_Axe_Skeleton
graphs [
 AnimGraph {nodes [
   n1 = Use cached pose 'equipped pose'()
   n2 = Use cached pose 'unequipped pose'()
   n3 = Get IsHoldingPickaxe() -> (IsHoldingPickaxe:bool)
   Output Pose(Result=Blend Poses by bool(
     BlendPose_0=Blend Poses by bool(BlendPose_0=@n1, BlendPose_1=@n2, bActiveValue=@n3) {blendType=Linear}
     BlendPose_1=Use cached pose 'main pose'()
     bActiveValue=Get bIsInFrontEnd()
    ) {blendType=Linear}) {layerGroup=None}
   AnimGraphNode_SubInput_0 = Input Pose()
   Save cached pose 'start pose'(Pose=@AnimGraphNode_SubInput_0) {cacheName=start pose}
   n4 = Anim Dynamics - Bone: dyn_upper_blade_01!(ComponentPose=Local To Component(LocalPose=Use cached pose 'start pose'()), Alpha=1, GravityOverride:Vector="0,0,0") -> (Pose) {
    angularDampingOverride=1
    angularSpringConstant=2000
    bAngularSpring=true
    bOverrideAngularDamping=true
    bUseGravityOverride=true
    bUsePlanarLimit=false
    simulationSpace=RootRelative
    alphaBoolBlend {blendOption=Linear}
    boundBone {boneName=dyn_upper_blade_01}
    componentAppliedLinearAccClamp=<700 700 700>
    componentLinearAccScale=<0.9 0.9 0.9>
    componentLinearVelScale=<0.9 0.9 0.9>
    physicsBodyDefinitions [{
      boundBone {boneName=dyn_upper_blade_01}
      boxExtents=<14 14 14>
      constraintSetup {angularLimitsMax=<0 15 0>, angularLimitsMin=<0 -20 0>, angularTarget=<1 0 0>}
      localJointOffset=<0.5 0 0>
     }]
   }
   n5 = Anim Dynamics - Bone: dyn_upper_blade_02!(ComponentPose=@n4, Alpha=1, GravityOverride:Vector="0,0,0") -> (Pose) {
    angularDampingOverride=1
    angularSpringConstant=2000
    bAngularSpring=true
    bOverrideAngularDamping=true
    bUseGravityOverride=true
    bUsePlanarLimit=false
    simulationSpace=RootRelative
    alphaBoolBlend {blendOption=Linear}
    boundBone {boneName=dyn_upper_blade_02}
    componentAppliedLinearAccClamp=<700 700 700>
    componentLinearAccScale=<0.9 0.9 0.9>
    componentLinearVelScale=<0.9 0.9 0.9>
    physicsBodyDefinitions [{
      boundBone {boneName=dyn_upper_blade_02}
      boxExtents=<12 12 12>
      constraintSetup {angularLimitsMax=<0 15 0>, angularLimitsMin=<0 -20 0>, angularTarget=<1 0 0>}
      localJointOffset=<0.5 0 0>
     }]
   }
   n6 = Anim Dynamics - Bone: dyn_upper_blade_03!(ComponentPose=@n5, Alpha=1, GravityOverride:Vector="0,0,0") -> (Pose) {
    angularDampingOverride=1
    angularSpringConstant=2000
    bAngularSpring=true
    bOverrideAngularDamping=true
    bUseGravityOverride=true
    bUsePlanarLimit=false
    simulationSpace=RootRelative
    alphaBoolBlend {blendOption=Linear}
    boundBone {boneName=dyn_upper_blade_03}
    componentAppliedLinearAccClamp=<700 700 700>
    componentLinearAccScale=<0.9 0.9 0.9>
    componentLinearVelScale=<0.9 0.9 0.9>
    physicsBodyDefinitions [{
      boundBone {boneName=dyn_upper_blade_03}
      boxExtents=<10 10 10>
      constraintSetup {angularLimitsMax=<0 15 0>, angularLimitsMin=<0 -20 0>, angularTarget=<1 0 0>}
      localJointOffset=<0.5 0 0>
     }]
   }
   Save cached pose 'main pose'(Pose=Component To Local(ComponentPose=@n6)) {cacheName=main pose}
   n7 = Anim Dynamics - Bone: dyn_upper_blade_01(ComponentPose=Local To Component(LocalPose=Use cached pose 'start pose'()), Alpha=1, GravityOverride="0,0,0") {
    angularDampingOverride=1
    angularSpringConstant=2000
    bAngularSpring=true
    bOverrideAngularDamping=true
    bUseGravityOverride=true
    bUsePlanarLimit=false
    simulationSpace=RootRelative
    alphaBoolBlend {blendOption=Linear}
    boundBone {boneName=dyn_upper_blade_01}
    componentAppliedLinearAccClamp=<1100 1100 1100>
    componentLinearAccScale=<0.9 0.9 0.9>
    componentLinearVelScale=<0.9 0.9 0.9>
    physicsBodyDefinitions [{
      boundBone {boneName=dyn_upper_blade_01}
      boxExtents=<16 16 16>
      constraintSetup {angularLimitsMax=<0 15 0>, angularLimitsMin=<0 -20 0>, angularTarget=<1 0 0>}
      localJointOffset=<0.3 0 0>
     }]
   }
   n8 = Anim Dynamics - Bone: dyn_upper_blade_02(ComponentPose=@n7, Alpha=1, GravityOverride="0,0,0") {
    angularDampingOverride=1
    angularSpringConstant=2000
    bAngularSpring=true
    bOverrideAngularDamping=true
    bUseGravityOverride=true
    bUsePlanarLimit=false
    simulationSpace=RootRelative
    alphaBoolBlend {blendOption=Linear}
    boundBone {boneName=dyn_upper_blade_02}
    componentAppliedLinearAccClamp=<1100 1100 1100>
    componentLinearAccScale=<0.9 0.9 0.9>
    componentLinearVelScale=<0.9 0.9 0.9>
    physicsBodyDefinitions [{
      boundBone {boneName=dyn_upper_blade_02}
      boxExtents=<14 14 14>
      constraintSetup {angularLimitsMax=<0 15 0>, angularLimitsMin=<0 -20 0>, angularTarget=<1 0 0>}
      localJointOffset=<0.2 0 0>
     }]
   }
   n9 = Anim Dynamics - Bone: dyn_upper_blade_03(ComponentPose=@n8, Alpha=1, GravityOverride="0,0,0") {
    angularDampingOverride=1
    angularSpringConstant=2000
    bAngularSpring=true
    bOverrideAngularDamping=true
    bUseGravityOverride=true
    bUsePlanarLimit=false
    simulationSpace=RootRelative
    alphaBoolBlend {blendOption=Linear}
    boundBone {boneName=dyn_upper_blade_03}
    componentAppliedLinearAccClamp=<1100 1100 1100>
    componentLinearAccScale=<0.9 0.9 0.9>
    componentLinearVelScale=<0.9 0.9 0.9>
    physicsBodyDefinitions [{
      boundBone {boneName=dyn_upper_blade_03}
      boxExtents=<12 12 12>
      constraintSetup {angularLimitsMax=<0 15 0>, angularLimitsMin=<0 -20 0>, angularTarget=<1 0 0>}
      localJointOffset=<0.1 0 0>
     }]
   }
   Save cached pose 'equipped pose'(Pose=Component To Local(ComponentPose=@n9)) {cacheName=equipped pose}
   n10 = Anim Dynamics - Bone: dyn_upper_blade_01(ComponentPose=Local To Component(LocalPose=Use cached pose 'start pose'()), Alpha=1, GravityOverride="0,0,0") {
    angularDampingOverride=1
    angularSpringConstant=1500
    bAngularSpring=true
    bOverrideAngularDamping=true
    bUseGravityOverride=true
    bUsePlanarLimit=false
    alphaBoolBlend {blendOption=Linear}
    boundBone {boneName=dyn_upper_blade_01}
    physicsBodyDefinitions [{
      boundBone {boneName=dyn_upper_blade_01}
      boxExtents=<14 14 14>
      constraintSetup {angularLimitsMax=<0 15 0>, angularLimitsMin=<0 -20 0>, angularTarget=<1 0 0>}
      localJointOffset=<0.15 0 0>
     }]
    simSpaceSettings {maxAngularAcceleration=100, maxAngularVelocity=20, simSpaceAngularAlpha=1}
   }
   n11 = Anim Dynamics - Bone: dyn_upper_blade_02(ComponentPose=@n10, Alpha=1, GravityOverride="0,0,0") {
    angularDampingOverride=1
    angularSpringConstant=1500
    bAngularSpring=true
    bOverrideAngularDamping=true
    bUseGravityOverride=true
    bUsePlanarLimit=false
    alphaBoolBlend {blendOption=Linear}
    boundBone {boneName=dyn_upper_blade_02}
    physicsBodyDefinitions [{
      boundBone {boneName=dyn_upper_blade_02}
      boxExtents=<12 12 12>
      constraintSetup {angularLimitsMax=<0 15 0>, angularLimitsMin=<0 -20 0>, angularTarget=<1 0 0>}
      localJointOffset=<0.1 0 0>
     }]
    simSpaceSettings {maxAngularAcceleration=100, maxAngularVelocity=20, simSpaceAngularAlpha=1}
   }
   n12 = Anim Dynamics - Bone: dyn_upper_blade_03(ComponentPose=@n11, Alpha=1, GravityOverride="0,0,0") {
    angularDampingOverride=1
    angularSpringConstant=1500
    bAngularSpring=true
    bOverrideAngularDamping=true
    bUseGravityOverride=true
    bUsePlanarLimit=false
    alphaBoolBlend {blendOption=Linear}
    boundBone {boneName=dyn_upper_blade_03}
    physicsBodyDefinitions [{
      boundBone {boneName=dyn_upper_blade_03}
      boxExtents=<10 10 10>
      constraintSetup {angularLimitsMax=<0 15 0>, angularLimitsMin=<0 -20 0>, angularTarget=<1 0 0>}
      localJointOffset=<0.05 0 0>
     }]
    simSpaceSettings {maxAngularAcceleration=100, maxAngularVelocity=20, simSpaceAngularAlpha=1}
   }
   Save cached pose 'unequipped pose'(Pose=Component To Local(ComponentPose=@n12)) {cacheName=unequipped pose}
  ]}
 EventGraph {nodes [
   Event BlueprintUpdateAnimation()
   > n1 = Get!(execute:Exec) -> (then:Exec, else:Exec, BodyMesh:Skeletal Mesh Component Budgeted Object Reference)
   > Set IsHoldingPickaxe(IsHoldingPickaxe:bool=true) -> (Output_Get:bool)
   Set IsHoldingPickaxe(execute=@n1.else, IsHoldingPickaxe=false)
   n2 = Event OnCharacterCustomizationLoaded()
   > Set BodyMesh(BodyMesh:Skeletal Mesh Component Budgeted Object Reference=Get Array Item with EFort Custom Part Type(Array=@n2.CustomMeshParts)) -> (Output_Get:Skeletal Mesh Component Budgeted Object Reference)
   TryGetPawnOwner()
  ]}
]
variables [
 BodyMesh {type=SkeletalMeshComponentBudgeted}
 IsHoldingPickaxe {type=boolean}
]
